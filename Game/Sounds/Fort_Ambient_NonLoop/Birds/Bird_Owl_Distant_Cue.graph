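# Sound cue for a distant owl one-shot in the non-looping ambient bird sounds.
# A quality level node selects between a random pool of eight distant and hoot waves and a lighter three-wave pool, with two further slots empty.
# Each pool feeds a modulator that allows deep downward pitch variation.
# Playback uses one-shot tile ambient attenuation and overrides concurrency to prevent new instances at the limit.
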SoundCue /Game/Sounds/Fort_Ambient_NonLoop/Birds/Bird_Owl_Distant_Cue
root=QualityLevel_0
nodes [
 {id=Modulator_1, children [Random_15], properties {pitchMin=0.5}}
 {id=Modulator_2, children [Random_0], properties {pitchMin=0.5}}
 {id=QualityLevel_0, children [Modulator_1,Modulator_2,null,null]}
 {id=Random_0, children [WavePlayer_2,WavePlayer_5,WavePlayer_8], properties {weights [1,1,1]}}
 {
  id=Random_15
  children [WavePlayer_128, WavePlayer_1, WavePlayer_136, WavePlayer_137, WavePlayer_4, WavePlayer_7, WavePlayer_134, WavePlayer_135]
  properties {weights [1,1,1,1,1,1,1,1]}
 }
 {id=WavePlayer_1, properties {soundWaveAssetPtr=&/Bird_OwlDistant01}}
 {id=WavePlayer_128, properties {soundWaveAssetPtr=&/Bird_OwlDistant08}}
 {id=WavePlayer_134, properties {soundWaveAssetPtr=&/Bird_OwlDistant06}}
 {id=WavePlayer_135, properties {soundWaveAssetPtr=&/Bird_OwlDistant07}}
 {id=WavePlayer_136, properties {soundWaveAssetPtr=&/Bird_OwlHoot02}}
 {id=WavePlayer_137, properties {soundWaveAssetPtr=&/Bird_OwlHoot01}}
 {id=WavePlayer_2, properties {soundWaveAssetPtr=&Bird_OwlDistant01}}
 {id=WavePlayer_4, properties {soundWaveAssetPtr=&/Bird_OwlDistant04}}
 {id=WavePlayer_5, properties {soundWaveAssetPtr=&Bird_OwlDistant04}}
 {id=WavePlayer_7, properties {soundWaveAssetPtr=&/Bird_OwlDistant05}}
 {id=WavePlayer_8, properties {soundWaveAssetPtr=&Bird_OwlDistant05}}
]
properties {
 attenuationSettings=/Game/Sounds/Attenuation/Fort_TileAmbient_OneShot_Att
 bOverrideConcurrency=true
 maxDistance=8000
 soundClassObject=/Game/Sounds/FortSoundClassesAndModes/Fort_Ambient_NonLoop
 volumeMultiplier=1.25
 concurrencyOverrides {resolutionRule=PreventNew}
}
